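; Camera effect Blueprint that plays a timed post-process radial blur centered on the first player's pawn.
; On BeginPlay it creates a dynamic instance of the world-position radial blur material and installs it as a full-weight blendable on its high-priority PostProcess component.
; A timeline then drives the Blend and BlurControl parameters, updating WorldPosition from the pawn's mesh location on each update.
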(event EventTick
  (params (DeltaSeconds :type Float%20%28single-precision%29)))

(event EventBeginPlay
  (bind _postprocess (Variables|Default|GetPostProcess))
  (bind _returnvalue (Rendering|Material|CreateDynamicMaterialInstance "/Game/Packages/Fortnite_UniqueMaterials_1/PostProcess/RadialBlur/PPM_RadialBlur_WorldPosition.PPM_RadialBlur_WorldPosition" "None" "None" :!class /Script/Engine.KismetMaterialLibrary))
  (Variables|Default|SetDM _returnvalue)
  (Variables|/Script/Engine.PostProcessComponent|SetSettings (Utilities|Struct|MakePostProcessSettings (Utilities|Struct|MakeWeightedBlendables (Utilities|Array|MakeArray (Utilities|Struct|MakeWeightedBlendable 1.0 _returnvalue)))) _postprocess)
  (Variables|/Script/Engine.PostProcessComponent|SetPriority 1e+02 _postprocess)
  (bind (_direction _newtrack_0 _newtrack_1) (Timeline|Timeline PlayFromStart)
    (:Update
      (Rendering|Material|SetScalarParameterValue _returnvalue "Blend" _newtrack_0 :!class /Script/Engine.MaterialInstanceDynamic)
      (Rendering|Material|SetVectorParameterValue _returnvalue "WorldPosition" (Math|Conversions|ToLinearColor(Vector) (Transformation|GetWorldLocation (Variables|/Script/Engine.Character|GetMesh (Utilities|Casting|CastToFortPawn (Game|GetPlayerPawn 0))))) :!class /Script/Engine.MaterialInstanceDynamic)
      (Rendering|Material|SetScalarParameterValue _returnvalue "BlurControl" _newtrack_1 :!class /Script/Engine.MaterialInstanceDynamic))))

(fn ConstructionScript)
