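; Camera lens effect Blueprint that, by its name, produces a screen flash with chromatic aberration through a post-process component.
; On BeginPlay it sets the component's priority to game time plus a fixed offset, so later-spawned instances outrank earlier ones.
; A timeline then plays from the start and drives the post-process blend weight on each update.
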(event EventBeginPlay
  (bind _postprocess (Variables|Default|GetPostProcess))
  (Variables|/Script/Engine.PostProcessComponent|SetPriority (+ (Utilities|Time|GetGameTimeinSeconds) 1e+02) _postprocess)
  (bind (_direction _newtrack_0) (Timeline|Timeline PlayFromStart)
    (:Update
      (Variables|/Script/Engine.PostProcessComponent|SetBlendWeight _newtrack_0 _postprocess))))

(fn ConstructionScript)
