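; Camera lens effect Blueprint that shows a screen warning through a post-process component.
; The InitialPostMaterial custom event creates a dynamic instance of the ToyFox screen haze warning material, stores it and installs it as a full-weight post-process blendable.
; On BeginPlay, a timeline lerps a post-process setting between two fixed values and reapplies the component's blend weight from the Blend variable each update.
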(event EventBeginPlay
  (bind _postprocess (Variables|Default|GetPostProcess))
  (bind (_direction _newtrack_0) (Timeline|Timeline PlayFromStart)
    (:Update
      (bind _structout (Utilities|Struct|SetmembersinPostProcessSettings (Variables|/Script/Engine.PostProcessComponent|GetSettings _postprocess) (Math|Float|Lerp 1.3 3.6 _newtrack_0)))
      (Variables|/Script/Engine.PostProcessComponent|SetSettings _structout _postprocess)
      (Variables|/Script/Engine.PostProcessComponent|SetBlendWeight (Variables|Default|GetBlend) _postprocess))))

(event Custom|InitialPostMaterial
  (bind _postprocess (Variables|Default|GetPostProcess))
  (bind _returnvalue (Rendering|Material|CreateDynamicMaterialInstance "/Game/Effects/Fort_Effects/GameplayPulginsAssets/ToyFox/M_ScreenHazeWarning.M_ScreenHazeWarning" "None" "None" :!class /Script/Engine.KismetMaterialLibrary))
  (bind _output_get (Variables|Default|SetDynamicMaterial _returnvalue))
  (bind _structout (Utilities|Struct|SetmembersinPostProcessSettings (Variables|/Script/Engine.PostProcessComponent|GetSettings _postprocess) (Utilities|Struct|MakeWeightedBlendables (Utilities|Array|MakeArray (Utilities|Struct|MakeWeightedBlendable 1.0 _output_get)))))
  (Variables|/Script/Engine.PostProcessComponent|SetSettings _structout _postprocess))

(fn ConstructionScript)
