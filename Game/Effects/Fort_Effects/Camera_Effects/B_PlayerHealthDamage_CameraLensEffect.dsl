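; Camera lens effect Blueprint for player health damage that controls the brightness of its screen particle effect.
; A custom event, PassParticle_Parameter, writes the "brightness" float parameter on the emitter's particle system component.
; A first hit uses a fixed high brightness; otherwise brightness rises linearly as the incoming value, apparently health out of 100, falls, minus a small offset.
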(event Custom|PassParticle_Parameter
  (params (NewParam :type Float%20%28double-precision%29) (First_Hit :type Boolean))
  (if First_Hit
    (Effects|Components|ParticleSystem|SetFloatParameter (Variables|Emitter|GetParticleSystemComponent) "brightness" 2)
    (else
      (Effects|Components|ParticleSystem|SetFloatParameter (Variables|Emitter|GetParticleSystemComponent) "brightness" (- (/ (- 100 NewParam) 100) 0.1)))))

(fn ConstructionScript)
