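; Camera lens effect Blueprint for player shield damage, built on a particle emitter.
; A custom event, PassParticle_Parameter_Shield, takes a number and sets the particle system's "brightness" parameter from it.
; The value is inverted against 100, normalized and shifted down by a fixed offset without clamping, so lower inputs give a brighter effect and high inputs go negative.
; The construction script only calls its parent.
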(event Custom|PassParticle_Parameter_Shield
  (params (NewParam :type Float%20%28double-precision%29))
  (Effects|Components|ParticleSystem|SetFloatParameter (Variables|Emitter|GetParticleSystemComponent) "brightness" (- (/ (- 100 NewParam) 100) 0.35)))

(fn ConstructionScript
  (|CallParentFunction))
