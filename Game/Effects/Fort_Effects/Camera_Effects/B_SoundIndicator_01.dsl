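; Sound indicator Blueprint that supplies the look and strength falloff for on-screen audio visualizer cues such as footsteps, gunfire, chests, gliders, vehicles and healing.
; Select functions map the indicator type to a default icon, tint and falloff curve, with some types sharing curves.
; Gunshots defer to GetWeaponCurve, which picks a melee, mid-range or long-range curve from the max audible distance on spawn.
; Infected reuses footstep icon and curve but sets no tint.
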(fn GetDefaultIcon
  (return (Utilities|Select :Index (Variables|Effects|GetIndicatorType) :Generic "/Game/UI/Foundation/Textures/Icons/Navigation/t_icon_nav_mobile_audioviz_genericsound_32.t_icon_nav_mobile_audioviz_genericsound_32" :FootStep "/Game/UI/Foundation/Textures/Icons/Navigation/t_icon_nav_mobile_audioviz_footsteps_32.t_icon_nav_mobile_audioviz_footsteps_32" :Gunshot "/Game/UI/Foundation/Textures/Icons/Navigation/t_icon_nav_mobile_audioviz_gunfire_32.t_icon_nav_mobile_audioviz_gunfire_32" :Chest "/Game/UI/Foundation/Textures/Icons/Navigation/t_icon_nav_mobile_audioviz_chest_32.t_icon_nav_mobile_audioviz_chest_32" :Glider "/Game/UI/Foundation/Textures/Icons/Navigation/t_icon_nav_mobile_audioviz_glider_32.t_icon_nav_mobile_audioviz_glider_32" :Vehicle "/Game/UI/Foundation/Textures/Icons/Navigation/ATK_visualizer_icon_32.ATK_visualizer_icon_32" :Infected "/Game/UI/Foundation/Textures/Icons/Navigation/t_icon_nav_mobile_audioviz_footsteps_32.t_icon_nav_mobile_audioviz_footsteps_32" :Healing "/Game/UI/Foundation/Textures/Icons/Navigation/T_Icon_Sound_MedConsumable_128.T_Icon_Sound_MedConsumable_128")))

(fn GetDefaultTint
  (return (Utilities|Select :Index (Variables|Effects|GetIndicatorType) :Generic "(R=0.262745,G=0.364706,B=0.490196,A=1.000000)" :FootStep "(R=0.494118,G=0.352941,B=0.223529,A=1.000000)" :Gunshot "(R=1.000000,G=0.431373,B=0.231373,A=1.000000)" :Chest "(R=1.000000,G=0.490196,B=0.039216,A=1.000000)" :Glider "(R=0.701961,G=0.662745,B=0.164706,A=1.000000)" :Vehicle "(R=0.262745,G=0.364706,B=0.490196,A=1.000000)" :Healing "(R=0.043137,G=0.615686,B=0.000000,A=1.000000)")))

(fn GetStrengthCurveForActiveType
  (return (Utilities|Select :Index (Variables|Effects|GetIndicatorType) :Generic (Variables|Falloff|GetFootsteps%20Falloff) :FootStep (Variables|Falloff|GetFootsteps%20Falloff) :Gunshot (CallFunction|GetWeaponCurve) :Chest (Variables|Falloff|GetChest%20Falloff) :Glider (Variables|Falloff|GetGlider%20Falloff) :Vehicle (Variables|Falloff|GetPlane%20Falloff) :Infected (Variables|Falloff|GetFootsteps%20Falloff) :Healing (Variables|Falloff|GetChest%20Falloff))))

(fn GetWeaponCurve
  (return (select (> (Variables|Effects|GetMaxAudibleDistanceOnSpawn) 2e+04) (Variables|Falloff|GetGunshot%20Falloff%20Long%20Range) (select (<= (Variables|Effects|GetMaxAudibleDistanceOnSpawn) 4e+03) (Variables|Falloff|GetGunshot%20Falloff%20Melee) (Variables|Falloff|GetGunshot%20Falloff%20Mid%20Range)))))

(fn ConstructionScript)
