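# Niagara camera lens effect for a big leaf pile, showing a burst of orange-brown bokeh sprites.
# The GPU Bokeh emitter spawns offset sprites on a squashed ring, while two leaf emitters with shrinking sprites and animated radii are disabled.
# Two system Timeline values drive alpha curves that fade the bokeh in and out.
# User.Exit blends acceleration from a random downward pull at 0 to zero at its default of 1.
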NiagaraSystem /Game/Effects/Fort_Effects/Camera_Effects/Effects/NS_Camera_Lens_BigLeafPile
dependencies {
 usedDataInterfaces [&NiagaraDataInterfaceCurve]
 usedDynamicInputs []
 usedModules [&SystemState, &Timeline]
}
emitters {
 Bokeh {
  simTarget=GPUComputeSim
  emitterUpdateScript [
   &EmitterState(
    Inactive Response=::NewEnumerator1
    Life Cycle Mode=::NewEnumerator1
    Loop Behavior=::NewEnumerator1
    Loop Duration Mode=::NewEnumerator1
    UseLoopDelay=false
   )
   &SpawnBurst_Instantaneous(Spawn Count=15, Use Spawn Probability=false, UseLoopCountLimit=false)
  ]
  particleSpawnScript [
   &InitializeParticle(
    Color=&RandomRangeLinearColor(Minimum=<rgba 0.3372 0.0901 0.0282 1>, Maximum=<rgba 0.8958 0.2503 0.1008 1>)
    Position Offset=<-90 0 0>
    Sprite Rotation Mode=::NewEnumerator2
    Sprite Size Mode=::NewEnumerator1
    Sprite UV Mode=::NewEnumerator3
    Uniform Sprite Size Max=100
    Uniform Sprite Size Min=50
    UsePositionOffset=true
    Write Material Random=true
   )
   &ShapeLocation(
    Handle Radius=30
    Large Radius=180
    MaskSpawnGroup=false
    Non Uniform Scale=<0.6 1 1>
    Override Seed=false
    Rotation Mode=::NewEnumerator1
    Shape Primitive=::NewEnumerator3
    UseOwnerScale=true
   ) {module=ShapeLocation002}
   /SetVariables() {isSetParameters=true}
  ]
  particleUpdateScript [
   &ParticleState(
    Kill Particles When Lifetime Has Elapsed=false
    Let Infinitely Lived Particles Die When Emitter Deactivates=true
    Loop Particles Lifetime=true
   )
   &AccelerationForce(Acceleration=&Lerp_Vector(A=&RandomRangeVector(Minimum=<0 0 -800>, Maximum=<1 1 -1000>), B=<0 0 0>, Alpha=$User.Exit))
   &DynamicMaterialParameters(
    Index 0 Param 1=0.1
    Module.Param0WriteEnabled=true
    Module.Param1WriteEnabled=false
    Module.Param2WriteEnabled=false
    Module.Param3WriteEnabled=false
   )
   &ScaleSpriteSize(
    Scale Sprite Size Mode=::NewEnumerator0
    Uniform Scale Factor=&Lerp_Float(A=1, B=0, Alpha=$System.Timeline.NormalizedTime)
   ) {enabled=false}
   &ScaleColor(
    Scale Alpha=&FloatFromCurve(
     FloatCurve=DataInterface() {propertyValues {Curve {keys [
         {interpMode=RCIM_Cubic, tangentMode=RCTM_User, time=0, value=0}
         {interpMode=RCIM_Cubic, tangentMode=RCTM_User, time=0.4, value=1}
         <at 0.4 1>
        ]}}}
     CurveIndex=$System.Timeline001.NormalizedTime
     Scale Curve=0.3
    )
    ScaleA=true
    ScaleRGB=false
   )
   &SolveForcesAndVelocity(Clamp Velocity=false, Limit Acceleration=false)
   &ScaleColor(
    Scale Alpha=&Lerp_Float(
     A=&FloatFromCurve(
      FloatCurve=DataInterface() {propertyValues {Curve {keys [{interpMode=RCIM_Cubic, leaveTangent=-4.5411, tangentMode=RCTM_Break, time=0, value=1}, <at 0.4 0>]}}}
      CurveIndex=$System.Timeline.NormalizedTime
     )
     B=0
     Alpha=$System.Timeline.NormalizedTime
    )
    ScaleA=true
    ScaleRGB=false
   ) {module=ScaleColor002}
  ]
  rendererClasses [&NiagaraSpriteRendererProperties]
  renderers [{rendererClass=&NiagaraSpriteRendererProperties}]
 }
 Leaves {
  bEnabled=false
  simTarget=GPUComputeSim
  emitterUpdateScript [
   &EmitterState(
    Inactive Response=::NewEnumerator1
    Life Cycle Mode=::NewEnumerator1
    Loop Behavior=::NewEnumerator1
    Loop Duration Mode=::NewEnumerator1
    UseLoopDelay=false
   )
   &SpawnBurst_Instantaneous(Spawn Count=60, Use Spawn Probability=false, UseLoopCountLimit=false)
  ]
  particleSpawnScript [&InitializeParticle(
    Color=&RandomRangeLinearColor(Minimum=<rgba 0.3372 0.2195 0.0369 1>, Maximum=<rgba 0.8958 0.4472 0.07 1>)
    Position Offset=<-90 0 0>
    Sprite Rotation Mode=::NewEnumerator2
    Sprite Size Mode=::NewEnumerator1
    Sprite UV Mode=::NewEnumerator3
    Uniform Sprite Size Max=30
    Uniform Sprite Size Min=23
    UsePositionOffset=true
    Write Material Random=true
   )]
  particleUpdateScript [
   &ParticleState(
    Kill Particles When Lifetime Has Elapsed=false
    Let Infinitely Lived Particles Die When Emitter Deactivates=true
    Loop Particles Lifetime=true
   )
   &AccelerationForce(Acceleration=&Lerp_Vector(A=&RandomRangeVector(Minimum=<0 0 -800>, Maximum=<1 1 -2000>), B=<1 1 0>, Alpha=$User.Exit))
   &ScaleSpriteSize(
    Scale Sprite Size Mode=::NewEnumerator0
    Uniform Scale Factor=&Lerp_Float(A=1, B=0, Alpha=$System.Timeline.NormalizedTime)
   )
   &DynamicMaterialParameters(
    Index 0 Param 1=0.1
    Module.Param0WriteEnabled=true
    Module.Param1WriteEnabled=false
    Module.Param2WriteEnabled=false
    Module.Param3WriteEnabled=false
   )
   &ScaleColor(
    Scale Alpha=&FloatFromCurve(FloatCurve=DataInterface() {propertyValues {Curve {keys [
         {arriveTangent=3.1109, interpMode=RCIM_Cubic, leaveTangent=3.1109, tangentMode=RCTM_User, time=0, value=0}
         {interpMode=RCIM_Cubic, tangentMode=RCTM_User, time=1, value=1}
        ]}}})
    Scale RGB=&MakeVectorFromLinearColorRGB(Color=&HueShiftLinearColor(
      Hue=&RandomRangeFloat(Minimum=0, Maximum=0.1)
      Saturation=1
      Value=&RandomRangeFloat(Minimum=1, Maximum=2)
      Linear Color=$Particles.Initial.Color
     ))
    ScaleA=false
    ScaleRGB=true
   )
   &ShapeLocation(
    Handle Radius=30
    Large Radius=&FloatFromCurve(
     FloatCurve=DataInterface() {propertyValues {Curve {keys [
         {arriveTangent=-619.6776, interpMode=RCIM_Cubic, leaveTangent=-619.6776, tangentMode=RCTM_User, time=0, value=300}
         {arriveTangent=-125.8032, interpMode=RCIM_Cubic, leaveTangent=-125.8032, tangentMode=RCTM_User, time=0.4, value=100}
        ]}}}
     CurveIndex=$System.Timeline001.NormalizedTime
    )
    MaskSpawnGroup=false
    Non Uniform Scale=<0.5 1 1>
    Override Seed=false
    Rotation Mode=::NewEnumerator1
    Shape Primitive=::NewEnumerator3
    UseOwnerScale=true
   )
   &SolveForcesAndVelocity(Clamp Velocity=false, Limit Acceleration=false)
  ]
  rendererClasses [&NiagaraSpriteRendererProperties]
  renderers [{rendererClass=&NiagaraSpriteRendererProperties}]
 }
 Leaves003 {
  bEnabled=false
  simTarget=GPUComputeSim
  emitterUpdateScript [
   &EmitterState(
    Inactive Response=::NewEnumerator1
    Life Cycle Mode=::NewEnumerator1
    Loop Behavior=::NewEnumerator1
    Loop Duration Mode=::NewEnumerator1
    UseLoopDelay=false
   )
   &SpawnBurst_Instantaneous(Spawn Count=45, Use Spawn Probability=false, UseLoopCountLimit=false)
  ]
  particleSpawnScript [&InitializeParticle(
    Color=&RandomRangeLinearColor(Minimum=<rgba 0.3372 0.2195 0.0369 1>, Maximum=<rgba 0.8958 0.4472 0.07 1>)
    Position Offset=<-90 0 0>
    Sprite Rotation Mode=::NewEnumerator2
    Sprite Size Mode=::NewEnumerator1
    Sprite UV Mode=::NewEnumerator3
    Uniform Sprite Size Max=25
    Uniform Sprite Size Min=17
    UsePositionOffset=false
    Write Material Random=true
   )]
  particleUpdateScript [
   &ParticleState(
    Kill Particles When Lifetime Has Elapsed=false
    Let Infinitely Lived Particles Die When Emitter Deactivates=true
    Loop Particles Lifetime=true
   )
   &AccelerationForce(Acceleration=&Lerp_Vector(A=&RandomRangeVector(Minimum=<0 0 -950>, Maximum=<1 1 -9000>), B=<1 1 0>, Alpha=$User.Exit)) {module=AccelerationForce001}
   &ScaleSpriteSize(
    Scale Sprite Size Mode=::NewEnumerator0
    Uniform Scale Factor=&Lerp_Float(
     A=&FloatFromCurve(
      FloatCurve=DataInterface() {propertyValues {Curve {keys [{interpMode=RCIM_Cubic, tangentMode=RCTM_User, time=0, value=1}, {interpMode=RCIM_Cubic, tangentMode=RCTM_User, time=0.45, value=0}]}}}
      CurveIndex=$System.Timeline.NormalizedTime
     )
     B=1
     Alpha=$User.Exit
    )
   ) {module=ScaleSpriteSize002}
   &DynamicMaterialParameters(
    Index 0 Param 1=0.1
    Module.Param0WriteEnabled=true
    Module.Param1WriteEnabled=false
    Module.Param2WriteEnabled=false
    Module.Param3WriteEnabled=false
   )
   &ScaleColor(
    Scale Alpha=0.8
    Scale RGB=&MakeVectorFromLinearColorRGB(Color=&HueShiftLinearColor(
      Hue=&RandomRangeFloat(Minimum=0, Maximum=0.1)
      Saturation=1
      Value=&RandomRangeFloat(Minimum=1, Maximum=2)
      Linear Color=$Particles.Initial.Color
     ))
    ScaleA=true
    ScaleRGB=true
   )
   &ShapeLocation(
    Handle Radius=15
    Large Radius=&FloatFromCurve(
     FloatCurve=DataInterface() {propertyValues {Curve {keys [
         {arriveTangent=114.3521, interpMode=RCIM_Cubic, leaveTangent=114.3521, tangentMode=RCTM_User, time=0, value=150}
         {arriveTangent=-1141.5936, interpMode=RCIM_Cubic, leaveTangent=-1141.5936, tangentMode=RCTM_User, time=0.1, value=100}
        ]}}}
     CurveIndex=$System.Timeline001.NormalizedTime
    )
    MaskSpawnGroup=false
    Non Uniform Scale=<0.5 1 1>
    Override Seed=false
    Rotation Mode=::NewEnumerator1
    Shape Primitive=::NewEnumerator3
    UseOwnerScale=true
   )
   &SolveForcesAndVelocity(Clamp Velocity=false, Limit Acceleration=false)
  ]
  rendererClasses [&NiagaraSpriteRendererProperties]
  renderers [{rendererClass=&NiagaraSpriteRendererProperties}]
 }
}
summary {
 systemName=NS_Camera_Lens_BigLeafPile
 emitters [
  {emitterName=Bokeh, simTarget=GPUComputeSim, rendererClasses [&NiagaraSpriteRendererProperties]}
  {bEnabled=false, emitterName=Leaves, simTarget=GPUComputeSim, rendererClasses [&NiagaraSpriteRendererProperties]}
  {bEnabled=false, emitterName=Leaves003, simTarget=GPUComputeSim, rendererClasses [&NiagaraSpriteRendererProperties]}
 ]
 userVariables [
  User.Exit {type=NiagaraFloat, defaultValue=1, description=""}
  User.IsRemoving {type=NiagaraBool}
 ]
}
