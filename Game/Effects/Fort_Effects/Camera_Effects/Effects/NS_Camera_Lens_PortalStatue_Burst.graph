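# Niagara camera-lens effect that, going by its name, plays a short speed-line burst for portal statue travel.
# Its one sprite emitter pairs an instant burst with tapering continuous spawning on a long cylinder surface aligned to System.TravelSpeed.
# Travel-scaled velocity, drag and opposing wind move the blue streaks, which stretch with particle speed.
# System.NormalizedSpeedScale shortens lifetime and drives brightness, while curves animate alpha and a material parameter.
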NiagaraSystem /Game/Effects/Fort_Effects/Camera_Effects/Effects/NS_Camera_Lens_PortalStatue_Burst
dependencies {
 usedDataInterfaces [&NiagaraDataInterfaceCurve]
 usedDynamicInputs [
  &MakeVector
  &RemapRange
  &VectorLength
 ]
 usedModules [
  &self:SystemSpawnScript.SystemScriptSource.SystemScriptGraph.NiagaraNodeAssignment_3.SetVariables_#1
  &self:SystemSpawnScript.SystemScriptSource.SystemScriptGraph.NiagaraNodeAssignment_5.SetVariables_#2
  &SystemState
  &self:SystemSpawnScript.SystemScriptSource.SystemScriptGraph.NiagaraNodeAssignment_4.SetVariables_#3
  &self:SystemSpawnScript.SystemScriptSource.SystemScriptGraph.NiagaraNodeAssignment_7.SetVariables_#4
  &self:SystemSpawnScript.SystemScriptSource.SystemScriptGraph.NiagaraNodeAssignment_0.SetVariables_#5
  &self:SystemSpawnScript.SystemScriptSource.SystemScriptGraph.NiagaraNodeAssignment_2.SetVariables_#6
  &self:SystemSpawnScript.SystemScriptSource.SystemScriptGraph.NiagaraNodeAssignment_1.SetVariables_#7
 ]
}
emitters {SpeedLinesBurst {
  emitterUpdateScript [
   &EmitterState(Life Cycle Mode=::NewEnumerator1, Loop Behavior=::NewEnumerator1, Loop Duration=0.4, UseLoopDelay=false)
   &SpawnBurst_Instantaneous(Spawn Count=250, Use Spawn Probability=false, UseLoopCountLimit=false)
   &SpawnRate(
    Spawn Probability=&FloatFromCurve(FloatCurve=DataInterface() {propertyValues {Curve {keys [<at 0 1>,<at 0.5 0>]}}}, CurveIndex=$Emitter.NormalizedLoopAge)
    SpawnRate=1000
    Use Spawn Probability=true
   )
  ]
  particleSpawnScript [
   /SetVariables(Particles.Intensity:NiagaraFloat=&FloatFromCurve(
     FloatCurve=DataInterface() {propertyValues {Curve {keys [
         {interpMode=RCIM_Cubic, tangentMode=RCTM_User, time=0, value=0}
         {arriveTangent=1.6995, interpMode=RCIM_Cubic, leaveTangent=1.6995, tangentMode=RCTM_User, time=0.5215, value=0.3748}
         {interpMode=RCIM_Cubic, tangentMode=RCTM_User, time=0.9, value=1}
        ]}}}
     CurveIndex=&RandomRangeFloat(Minimum=0, Maximum=1)
    )) {isSetParameters=true}
   &InitializeParticle(
    Color=&LerpLinearColors(StartColor=<rgba 0 0.8651 1 0.6>, EndColor=<rgba 0.2188 0.5108 1 0.8>, LerpFactor=$Particles.Intensity)
    Lifetime=&Multiply_Float(
     A=&RandomRangeFloat(Minimum=0.6, Maximum=1.3)
     B=&Lerp_Float(A=1, B=0.35, Alpha=$System.NormalizedSpeedScale)
    )
    Sprite Size=&Multiply_Vector2D_ByFloat(Float=0.25, Vector2D=&Lerp_Vector2(A=<70 60>, B=<150 75>, Alpha=$Particles.Intensity))
    Sprite Size Mode=::NewEnumerator4
    UsePositionOffset=false
    Write Material Random=true
   )
   &ShapeLocation(
    Cylinder Height=25000
    Cylinder Radius=150
    MaskSpawnGroup=false
    Offset=&Multiply_VectorByFloat(Vector=$System.TravelSpeed, Float=1)
    Offset Mode=::NewEnumerator0
    Override Seed=false
    Rotation Axis=$System.TravelSpeed
    Rotation Coordinate Space=::Simulation
    Rotation Mode=::NewEnumerator1
    Shape Primitive=::NewEnumerator1
    Surface Only=true
    Use Endcaps In Surface Only Mode=false
    UseOwnerScale=true
   )
   &AddVelocity(Velocity=$System.TravelSpeed, Velocity Speed Scale=5)
  ]
  particleUpdateScript [
   &ParticleState()
   &ScaleColor(
    Scale Alpha=&FloatFromCurve(FloatCurve=DataInterface() {propertyValues {Curve {keys [
         {interpMode=RCIM_Cubic, tangentMode=RCTM_User, time=0, value=0}
         <at 0.25 1>
         {interpMode=RCIM_Cubic, tangentMode=RCTM_User, time=0.75, value=1}
         {tangentMode=RCTM_User, time=1, value=0}
        ]}}})
    Scale RGB=&VectorFromFloat(Value=&FloatFromCurve(FloatCurve=DataInterface() {propertyValues {Curve {keys [<at 0.25 0.3 RCIM_Cubic>,<at 1 2>]}}}, CurveIndex=$System.NormalizedSpeedScale))
    ScaleA=true
    ScaleRGB=true
   )
   &DynamicMaterialParameters(
    Index 0 Param 1=&Lerp_Float(
     A=5
     B=2
     Alpha=&FloatFromCurve(FloatCurve=DataInterface() {propertyValues {Curve {keys [
          {interpMode=RCIM_Cubic, tangentMode=RCTM_User, time=0, value=0}
          {interpMode=RCIM_Cubic, tangentMode=RCTM_User, time=0.7, value=1}
          {tangentMode=RCTM_User, time=1, value=0}
         ]}}})
    )
    Index 0 Param 2=500
    Index 0 Param 3=1
    Module.Param0WriteEnabled=true
    Module.Param1WriteEnabled=true
    Module.Param2WriteEnabled=true
    Module.Param3WriteEnabled=false
   )
   &Drag(Drag=4, Use Linear Drag=true, Use Rotational Drag=false)
   &WindForce(Air Resistance=0.1, Wind Speed=$System.TravelSpeed, Wind Speed Scale=-20)
   &SolveForcesAndVelocity(Clamp Velocity=false, Limit Acceleration=false)
   &ScaleSpriteSizeBySpeed(
    Max Scale Factor=<5 180>
    Min Scale Factor=<5 0.7>
    Sample Scale Factor By Curve=true
    Scale Factor Curve=DataInterface() {propertyValues {Curve {keys [<at 0 0>, {arriveTangent=1, interpMode=RCIM_Cubic, leaveTangent=1, time=0.6, value=0.4}, <at 1 1>]}}}
    Velocity Threshold=14000
   )
  ]
  rendererClasses [&NiagaraSpriteRendererProperties]
  renderers [{rendererClass=&NiagaraSpriteRendererProperties}]
 }}
summary {
 systemName=NS_Camera_Lens_PortalStatue_Burst
 emitters [{emitterName=SpeedLinesBurst, rendererClasses [&NiagaraSpriteRendererProperties]}]
 userVariables [
  User.HeatOpacity {type=NiagaraFloat, defaultValue=0.4, description=""}
  User.HeatScale {type=NiagaraFloat, defaultValue=5, description=""}
  User.PanSpeed {type=NiagaraFloat, defaultValue=5, description=""}
  User.ScreenSizScalar {type=NiagaraFloat, defaultValue=1, description=""}
  User.SpawnRocks {type=NiagaraFloat, defaultValue=0.4, description=""}
  User.SpeedlineLength {type=NiagaraFloat, defaultValue=1, description=""}
  User.SpeedLineMulti {type=NiagaraFloat, defaultValue=1, description=""}
 ]
}
