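# Generic screen-space directional indicator Niagara system, judging by its name and Camera_Effects path, with emitters named Icon, OriginalBand, Vehicle, FootSteps, AttackWaves and Infected.
# User parameters expose tint, strength, wave intensity, icon texture, location and lifetime, and a material instance slot.
# The system uses SystemState and a spawn-time Set Variables module, with dynamic inputs for building and multiplying colors and vectors.
# No emitter lists any renderers here.
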NiagaraSystem /Game/Effects/Fort_Effects/Camera_Effects/Effects/NS_Screen_DirectionalIndicator_Generic_LW
dependencies {
 usedDataInterfaces []
 usedDynamicInputs [
  &MakeLinearColorFromVectorAndFloat
  &MultiplyLinearColor
  &MakeVector4
 ]
 usedModules [
  &self:SystemSpawnScript.SystemScriptSource.SystemScriptGraph.NiagaraNodeAssignment_1.SetVariables_#1
  &SystemState
 ]
}
emitters {
 AttackWaves {rendererClasses [], renderers []}
 FootSteps {rendererClasses [], renderers []}
 Icon {rendererClasses [], renderers []}
 Infected {rendererClasses [], renderers []}
 OriginalBand {rendererClasses [], renderers []}
 Vehicle {rendererClasses [], renderers []}
}
summary {
 systemName=NS_Screen_DirectionalIndicator_Generic_LW
 emitters [
  {emitterName=Icon, rendererClasses []}
  {emitterName=OriginalBand, rendererClasses []}
  {emitterName=Vehicle, rendererClasses []}
  {emitterName=FootSteps, rendererClasses []}
  {emitterName=AttackWaves, rendererClasses []}
  {emitterName=Infected, rendererClasses []}
 ]
 userVariables [
  User.//Tint {type=Vector3f, description="", defaultValue {struct=Vector3f, value=<0 0 0>}}
  User.IconLifetime {type=NiagaraFloat, defaultValue=0.8, description=""}
  User.IconLocation {type=NiagaraPosition, description="", defaultValue {}}
  User.MID {
   type=MaterialInterface
   description=""
   defaultValue {struct=NiagaraExt_VariableValue_Object, value {object=null, objectClass=/Script/Engine.MaterialInterface}}
  }
  User.NewIcon {type=Texture, description="", defaultValue {struct=NiagaraExt_VariableValue_Object, value {object=null, objectClass=/Script/Engine.Texture}}}
  User.Strength {type=NiagaraFloat, defaultValue=1, description=""}
  User.Test {type=Vector3f, description="", defaultValue {struct=Vector3f, value=<0 1 1>}}
  User.Tint {type=LinearColor, description="", defaultValue {struct=LinearColor, value=<rgba 1 1 1 1>}}
  User.WaveIntensity {type=NiagaraFloat, defaultValue=1, description=""}
 ]
}
