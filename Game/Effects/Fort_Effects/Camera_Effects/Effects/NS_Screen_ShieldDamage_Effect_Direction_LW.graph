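# Niagara system for a directional shield-damage screen effect, judging by its name and Camera_Effects location.
# It holds two emitters, New and test, with no renderers listed, and SystemState is the only module dependency.
# It exposes a User.Rotation vector defaulting to the X axis, presumably for hit direction, though nothing shown here consumes it.
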NiagaraSystem /Game/Effects/Fort_Effects/Camera_Effects/Effects/NS_Screen_ShieldDamage_Effect_Direction_LW
dependencies {usedDataInterfaces [], usedDynamicInputs [], usedModules [&SystemState]}
emitters {New {rendererClasses [], renderers []}, test {rendererClasses [], renderers []}}
summary {
 systemName=NS_Screen_ShieldDamage_Effect_Direction_LW
 emitters [{emitterName=New, rendererClasses []}, {emitterName=test, rendererClasses []}]
 userVariables [User.Rotation {type=Vector3f, description="", defaultValue {struct=Vector3f, value=<1 0 0>}}]
}
