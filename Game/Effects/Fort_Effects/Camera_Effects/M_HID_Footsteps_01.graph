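# Translucent unlit particle material for a footsteps camera effect, drawn without depth test after motion blur.
# Particle color feeds emissive through MF_FX_EmissiveAdjustments.
# Two soft SphereMasks sit side by side on horizontally tiled UVs, and a time-driven sine, biased and scaled, blends between them so the glow sways from one to the other.
# The blend passes through a Power and is multiplied by particle alpha.
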Material /Game/Effects/Fort_Effects/Camera_Effects/M_HID_Footsteps_01
graph {
 properties {
  bDisableDepthTest=true
  bEnableResponsiveAA=true
  bUsedWithMeshParticles=true
  bUsedWithNanite=true
  bUsedWithNiagaraMeshParticles=true
  blendMode=BLEND_Translucent
  pixelDepthOffsetMode=PDOM_Legacy
  shadingModel=MSM_Unlit
  translucencyPass=MTP_AfterMotionBlur
  usedShadingModels=Unlit
 }
 nodes [
  n1 = ParticleColor()
  MaterialFunctionCall(Emissive=@n1.RGB) {materialFunction=&MF_FX_EmissiveAdjustments}
  n2 = TextureCoordinate() {uTiling=2}
  n3 = Constant()
  n4 = SphereMask(A=@n2, B=Constant2Vector() {g=0.5, r=0.6}, Radius=Constant2Vector() {g=0.5, r=0.5}, Hardness=@n3)
  n5 = SphereMask(A=@n2, B=Constant2Vector() {g=0.5, r=1.5}, Radius=Constant2Vector() {g=0.6, r=0.5}, Hardness=@n3)
  n6 = ConstantBiasScale(None=Sine(None=Time() {bOverride_Period=true, period=1.4286}) {period=0.7}) {bias=1.5, scale=0.4}
  Multiply(A=@n1.A, B=Power(Base=LinearInterpolate(A=@n4, B=@n5, Alpha=@n6)))
 ]
}
