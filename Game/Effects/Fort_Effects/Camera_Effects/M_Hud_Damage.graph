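# Translucent unlit particle material for a HUD damage direction indicator, drawn without depth testing after motion blur.
# Particle color RGB feeds an emissive adjustment function, while the indicator texture multiplied by particle alpha is clamped as a mask.
# A Rotation scalar parameter drives a cheap rotate-about-world-axis function, turning the indicator through vertex offset rather than by rotating UVs.
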Material /Game/Effects/Fort_Effects/Camera_Effects/M_Hud_Damage
graph {
 properties {
  bDisableDepthTest=true
  bUsedWithMeshParticles=true
  bUsedWithNanite=true
  bUsedWithNiagaraMeshParticles=true
  blendMode=BLEND_Translucent
  pixelDepthOffsetMode=PDOM_Legacy
  shadingModel=MSM_Unlit
  translucencyPass=MTP_AfterMotionBlur
  usedShadingModels=Unlit
 }
 nodes [
  n1 = TextureCoordinate()
  n2 = ParticleColor()
  Clamp(None=Multiply(A=TextureSampleParameter2D(UVs=@n1) {parameterName=Texture, texture=/Game/UI/Fort_HUD/Damage/T_Damage_Direction_Indicator}.RGB, B=@n2.A))
  MaterialFunctionCall(Emissive=@n2.RGB) {materialFunction=&MF_FX_EmissiveAdjustments}
  MaterialFunctionCall!(Rotation Amount=ScalarParameter() {parameterName=Rotation}, PivotPoint, WorldPosition) {materialFunction=/Engine/Functions/Engine_MaterialFunctions02/WorldPositionOffset/RotateAboutWorldAxis_cheap}
 ]
}
