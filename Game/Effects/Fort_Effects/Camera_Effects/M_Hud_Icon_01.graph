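# Unlit translucent particle-sprite material that draws a HUD-style icon with depth testing off, defaulting to a generic sound indicator texture.
# It samples the icon alpha at center-scaled UVs plus an offset copy for a drop shadow, blending both over a background color tinted by particle color.
# A SphereMask adds a soft circular backing to opacity, scaled by particle alpha.
# A time-driven sine along the pixel-to-particle-center direction creates a pulsing wiggle offset.
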Material /Game/Effects/Fort_Effects/Camera_Effects/M_Hud_Icon_01
graph {
 properties {
  bDisableDepthTest=true
  bEnableResponsiveAA=true
  bUsedWithNanite=true
  bUsedWithNiagaraSprites=true
  bUsedWithParticleSprites=true
  blendMode=BLEND_Translucent
  pixelDepthOffsetMode=PDOM_Legacy
  shadingModel=MSM_Unlit
  translucencyPass=MTP_AfterMotionBlur
  usedShadingModels=Unlit
 }
 nodes [
  n1 = VectorParameter() {parameterName=BGColor, defaultValue=<rgba 0.03 0.03 0.03 0>}
  n2 = MaterialFunctionCall(Texture Scale=AppendVector(A=ScalarParameter() {defaultValue=0.8, parameterName=U}, B=ScalarParameter() {defaultValue=0.8, parameterName=V})) {materialFunction=&ScaleUVsByCenter}
  n3 = TextureSampleParameter2D(UVs=@n2.UVs) {parameterName=Icon, texture=/Game/UI/Foundation/Textures/Icons/Navigation/t_icon_nav_mobile_audioviz_genericsound_128}
  n4 = ComponentMask(None=VectorParameter() {parameterName=DropShadowBias, defaultValue=<rgba -0.04 -0.04 0 0>}.RGB) {g=true, r=true}
  n5 = Add(
   A=@n3.A
   B=TextureSampleParameter2D(UVs=Add(A=@n2.UVs, B=@n4) {constB=-0.02}) {parameterName=Icon, texture=&t_icon_nav_mobile_audioviz_genericsound_128}.A
  )
  n6 = Saturate(None=@n5)
  n7 = ParticleColor()
  MaterialFunctionCall(Emissive=Multiply(A=LinearInterpolate(A=@n1.RGB, B=@n3.A, Alpha=@n6) {constA=0.1}, B=@n7.RGB)) {materialFunction=&MF_FX_EmissiveAdjustments}
  n8 = TextureCoordinate()
  n9 = Constant() {r=0.5}
  n10 = Constant() {r=0.5}
  n11 = Constant() {r=0.6}
  Multiply(A=@n7.A, B=Add(A=@n5, B=SphereMask(A=@n8, B=@n9, Radius=@n10, Hardness=@n11)))
  n12 = Subtract(A=ParticlePositionWS(), B=WorldPosition().XYZ)
  n13 = ConstantBiasScale(None=Sine(None=Time() {bOverride_Period=true, period=0.6667}) {period=1.5})
  n14 = ScalarParameter() {defaultValue=0.3, parameterName=WiggleAmount}
  Reroute(None=Multiply(A=Normalize(VectorInput=@n12), B=Multiply(A=@n13, B=@n14)))
  TextureCoordinate()
 ]
}
