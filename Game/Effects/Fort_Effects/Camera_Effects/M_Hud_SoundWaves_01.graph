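# Translucent unlit particle material that draws an animated sound-wave line, drawn without depth testing after motion blur.
# A rounded sine of panned UVs, driven by dynamic Wave Speed and squared Wave Intensity, offsets V in square steps.
# A horizontal mask fades the offset and thins the line toward both ends, with thickness set by Waves Radius.
# Particle color feeds emissive through MF_FX_EmissiveAdjustments, and particle alpha scales the line mask.
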Material /Game/Effects/Fort_Effects/Camera_Effects/M_Hud_SoundWaves_01
graph {
 properties {
  bDisableDepthTest=true
  bEnableResponsiveAA=true
  bUsedWithMeshParticles=true
  bUsedWithNanite=true
  bUsedWithNiagaraMeshParticles=true
  blendMode=BLEND_Translucent
  pixelDepthOffsetMode=PDOM_Legacy
  shadingModel=MSM_Unlit
  translucencyPass=MTP_AfterMotionBlur
  usedShadingModels=Unlit
 }
 nodes [
  n1 = ParticleColor()
  MaterialFunctionCall(Emissive=@n1.RGB) {materialFunction=&MF_FX_EmissiveAdjustments}
  n2 = Constant()
  n3 = Multiply(
   A=TextureCoordinate()
   B=DynamicParameter() {defaultValue=<rgba 0.15 8 1 1>, paramNames [Wave Intensity, Wave Speed, Strength, Param4]}.Wave Speed
  )
  n4 = Sine(None=ComponentMask(None=Panner(Coordinate=@n3) {bFractionalPart=true, speedX=1}) {r=true})
  n5 = DynamicParameter() {defaultValue=<rgba 0.15 8 1 1>, paramNames [Wave Intensity, Wave Speed, Strength, Param4]}
  n6 = Multiply(A=Round(None=AppendVector(A=@n2, B=@n4)), B=Multiply(A=@n5.Wave Intensity, B=@n5.Wave Intensity)) {constB=0.1}
  n7 = SphereMask(A=Constant() {r=0.5}, B=ComponentMask(None=TextureCoordinate()) {r=true}, Radius=Constant() {r=0.48}, Hardness=Constant())
  n8 = ComponentMask(None=Add(A=TextureCoordinate(), B=Multiply(A=@n6, B=@n7)) {constB=0}) {g=true}
  n9 = Constant() {r=0.5}
  n10 = ScalarParameter() {defaultValue=0.2, parameterName=Waves Radius}
  n11 = LinearInterpolate(A=Divide(A=@n10), B=@n10, Alpha=@n7)
  n12 = Constant()
  Multiply(A=@n1.A, B=Multiply(A=SphereMask(A=@n8, B=@n9, Radius=@n11, Hardness=@n12), B=@n7))
 ]
}
