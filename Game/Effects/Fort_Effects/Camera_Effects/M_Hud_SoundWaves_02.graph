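# Unlit translucent particle material that draws an animated sound-wave band for HUD or camera effects, with depth test disabled and rendering after motion blur.
# A panned sine offsets UVs, driven by Wave Speed and squared Wave Intensity dynamic parameters.
# Sphere masks shape the band, taper it horizontally and add a central spot.
# Particle color feeds emissive through MF_FX_EmissiveAdjustments, while particle alpha scales the combined mask.
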Material /Game/Effects/Fort_Effects/Camera_Effects/M_Hud_SoundWaves_02
graph {
 properties {
  bDisableDepthTest=true
  bEnableResponsiveAA=true
  bUsedWithMeshParticles=true
  bUsedWithNanite=true
  bUsedWithNiagaraMeshParticles=true
  blendMode=BLEND_Translucent
  pixelDepthOffsetMode=PDOM_Legacy
  shadingModel=MSM_Unlit
  translucencyPass=MTP_AfterMotionBlur
  usedShadingModels=Unlit
 }
 nodes [
  n1 = ParticleColor()
  MaterialFunctionCall(Emissive=@n1.RGB) {materialFunction=&MF_FX_EmissiveAdjustments}
  n2 = Panner(Coordinate=Multiply(
    A=TextureCoordinate()
    B=DynamicParameter() {defaultValue=<rgba 0.05 8 1 1>, paramNames [Wave Intensity, Wave Speed, Strength, Param4]}.Wave Speed
   )) {speedX=2.5}
  n3 = AppendVector(A=Constant() {r=1}, B=Sine(None=ComponentMask(None=@n2) {r=true}) {period=2.5})
  n4 = DynamicParameter() {defaultValue=<rgba 0.05 8 1 1>, paramNames [Wave Intensity, Wave Speed, Strength, Param4]}
  n5 = Multiply(A=@n4.Wave Intensity, B=@n4.Wave Intensity)
  n6 = TextureCoordinate()
  n7 = TextureCoordinate()
  n8 = AppendVector(
   A=ScalarParameter() {group=KeyArt Texture, parameterName=Offset X, sliderMax=1, sortPriority=1}
   B=Divide(A=ScalarParameter() {group=KeyArt Texture, parameterName=Offset Y, sliderMax=1, sortPriority=2}, B=ComponentMask(None=@n7) {g=true})
  )
  n9 = SphereMask(
   A=Constant() {r=0.5}
   B=MaterialFunctionCall(UVs=Subtract(A=@n7, B=@n8), Texture Scale=Constant2Vector() {g=5.5, r=0.6}) {materialFunction=&ScaleUVsByCenter}.UVs
   Radius=Constant() {r=0.1}
   Hardness=Constant() {r=0.3}
  ) {attenuationRadius=0.2}
  n10 = SphereMask(A=Constant() {r=0.5}, B=ComponentMask(None=Subtract(A=@n6, B=@n9) {constA=-5}) {r=true}, Radius=Constant() {r=0.48}, Hardness=Constant())
  n11 = Add(A=TextureCoordinate(), B=Multiply(A=Multiply(A=@n3, B=@n5) {constB=0.1}, B=@n10)) {constB=0}
  n12 = ScalarParameter() {defaultValue=0.36, parameterName=Waves Radius}
  n13 = Divide(A=@n12)
  n14 = Multiply(
   A=SphereMask(A=ComponentMask(None=@n11) {g=true}, B=Constant() {r=0.5}, Radius=LinearInterpolate(A=@n13, B=@n12, Alpha=@n10), Hardness=Constant())
   B=@n10
  )
  n15 = Constant() {r=0.5}
  n16 = MaterialFunctionCall(UVs=TextureCoordinate(), Texture Scale=Constant2Vector() {g=4, r=5}) {materialFunction=&ScaleUVsByCenter}
  n17 = Constant() {r=0.1}
  n18 = Constant() {r=-8}
  Multiply(
   A=@n1.A
   B=Add(A=SmoothStep(Value=@n14) {constMax=1.2}, B=SphereMask(A=@n15, B=@n16.UVs, Radius=@n17, Hardness=@n18) {attenuationRadius=0.2})
  )
 ]
}
