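# Unlit translucent mesh-particle material for a vehicle sound direction indicator, drawn without depth testing and after motion blur.
# It subtracts an offset exponential radial gradient from the HUD damage-direction texture, then scales by particle alpha to form a clamped mask.
# Clamped particle color passes through the shared FX emissive-adjustment function.
# A Rotation scalar parameter feeds a cheap rotate-about-world-axis function.
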Material /Game/Effects/Fort_Effects/Camera_Effects/M_Vehicle_Sound_Indicator
graph {
 properties {
  bDisableDepthTest=true
  bUsedWithMeshParticles=true
  bUsedWithNanite=true
  bUsedWithNiagaraMeshParticles=true
  blendMode=BLEND_Translucent
  pixelDepthOffsetMode=PDOM_Legacy
  shadingModel=MSM_Unlit
  translucencyPass=MTP_AfterMotionBlur
  usedShadingModels=Unlit
 }
 nodes [
  n1 = TextureSample(UVs=TextureCoordinate()) {texture=/Game/UI/Fort_HUD/Damage/T_Damage_Direction_Indicator}
  n2 = MaterialFunctionCall(
   UVs=Multiply(A=TextureCoordinate(), B=Constant2Vector() {g=1, r=2})
   CenterPosition=Constant2Vector() {g=0.29, r=1}
   Radius=Constant() {r=0.1}
   Density=Constant() {r=20}
  ) {materialFunction=&RadialGradientExponential}
  n3 = ParticleColor()
  Clamp(None=Multiply(A=Subtract(A=@n1.RGB, B=@n2.RadialGradientExponential), B=@n3.A))
  MaterialFunctionCall(Emissive=Clamp(None=@n3.RGB)) {materialFunction=&MF_FX_EmissiveAdjustments}
  MaterialFunctionCall!(Rotation Amount=ScalarParameter() {parameterName=Rotation}, PivotPoint, WorldPosition) {materialFunction=/Engine/Functions/Engine_MaterialFunctions02/WorldPositionOffset/RotateAboutWorldAxis_cheap}
 ]
}
